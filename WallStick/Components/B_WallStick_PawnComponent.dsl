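; Pawn component for a wall stick ability, declaring hooks for local client setup, landing, jumping, movement mode changes and the IA_WallStick input.
; CheckObstacleAbove sphere-traces upward from the owner along world Z using the Pawn profile and a configurable length, ignoring the owner.
; The event handlers and the FindWall, HandleWallStickSucceeded and EndWallStickMovementMode functions show no logic, so only the obstacle check is implemented here.
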(event Custom|InitializeLocallyControlledClient_1)

(event Custom|OnPawnLanded_Event
  (params (Hit :type Hit%20Result%20Structure%20%28by%20ref%29)))

(event Custom|OnPawnJumped_Event)

(event Custom|MovementModeChangedDelegate_Event
  (params (Character :type Character%20Object%20Reference) (PrevMovementMode :type EMovementMode%20Enum :default "MOVE_None") (PreviousCustomMode :type Byte)))

(event EnhancedInputActionIA_WallStick
  (params (ActionValue :type Boolean) (ElapsedSeconds :type Float%20%28double-precision%29) (TriggeredSeconds :type Float%20%28double-precision%29) (InputAction :type Input%20Action%20Object%20Reference :default "/WallStick/Input/Actions/IA_WallStick.IA_WallStick")))

(fn CheckObstacleAbove
  (bind _outhit (Collision|SphereTraceByProfile (Transformation|GetActorLocation (Components|GetOwner)) (+ (Transformation|GetActorLocation (Components|GetOwner)) (* "0.000000,0.000000,1.000000" (Variables|Default|GetObstacleAboveTraceLength))) 1e+01 "Pawn" false (Utilities|Array|MakeArray (Components|GetOwner)) "ForDuration"))
  (return _outhit))

(fn EndWallStickMovementMode)

(fn FindWall)

(fn HandleWallStickSucceeded)
